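; Player-state component that tracks cumulative storm damage over a match to drive Storm Sickness.
; Damage gameplay events tagged as outside the safe zone add to a total, whose OnRep sends threshold progress and checks warning, infliction and kill thresholds.
; Those trigger a warning event, a server-applied gameplay effect, or a force-kill against exploits.
; Pawn death, minigame end and Storm Sickness mutator option changes drive resets or re-checks through bound delegates.
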(event Custom|OnDied
  (params (DamagedActor :type Actor%20Object%20Reference) (Damage :type Float%20%28single-precision%29) (InstigatedBy :type Controller%20Object%20Reference) (DamageCauser :type Actor%20Object%20Reference) (HitLocation :type Vector) (FHitComponent :type Primitive%20Component%20Object%20Reference) (BoneName :type Name) (Momentum :type Vector))
  (CallFunction|Reset (or (ScalableFloat|GetValueasBoolwithContext (Variables|Default|GetbAllowResetWarnAndSicknessOnDeath)) (and (CallFunction|IsRespawningAllowed) (ScalableFloat|GetValueasBoolwithContext (Variables|Default|GetbResetStormSicknessOnRespawn))))))

(event Custom|UnbindMinigame
  (bind _storedminigame (Variables|Default|GetStoredMinigame))
  (CallFunction|ResetAll)
  (Utilities|IsValid (Variables|Default|GetStoredMinigame)
    (:Is%20Valid
      (Minigame|UnbindEventfromMinigameEndedDelegate _storedminigame (EventDispatchers|CreateEvent :!event Custom|ResetAll))
      (StormSickness|UnbindEventfromOnStormSicknessOptionsUpdated (Default|GetStormSicknessMutator _storedminigame) (EventDispatchers|CreateEvent :!event Custom|OnOptionsUpdated)))))

(event Custom|OnOptionsUpdated
  (Utilities|EnabledCheck
    (:True
      (CallFunction|CheckAddStormSickness))
    (:False
      (CallFunction|ResetAbility))))

(event EventEndPlay
  (params (EndPlayReason :type EEndPlayReason%20Enum :default "Destroyed"))
  (bind _fortplayerpawn (Variables|Default|GetFortPlayerPawn))
  (CallFunction|UnbindMinigame)
  (Utilities|IsValid _fortplayerpawn
    (:Is%20Valid
      (EventDispatchers|UnbindEventfromOnDied _fortplayerpawn (EventDispatchers|CreateEvent :!event Custom|OnDied)))))

(event EventBeginPlay
  ;[0 1 0.2324] Tracking storm damage taken by the player over the course of a match
  (bind (_asynctaskproxy _payload) (Ability|Async|WaitGameplayEventToActor (Components|GetOwner) "(TagName=\"Event.Damage.Received\")")
    (:EventReceived
      (if (GameplayTags|TagContainer|CompareTagContainertoOtherTags (Variables|Default|GetOutsideSafeZoneDamageTag) (.InstigatorTags (Utilities|Struct|BreakGameplayEventData _payload)))
        (Utilities|EnabledCheck
          (:True
            (CallFunction|SendStormSicknessGameplayEvent _payload)
            (Development|PrintText "INVTEXT(\"Sent Storm Damage\")" true true "(R=0.000000,G=0.660000,B=1.000000,A=1.000000)" 1.0))
          (:False
            (CallFunction|ResetAbility)))))))

(event Custom|SendStormSicknessGameplayEvent
  (params (Gameplay%20Event%20Data :type Gameplay%20Event%20Data%20Structure))
  (bind (_eventtag _instigator _target _optionalobject _optionalobject2 _contexthandle _instigatortags _targettags _eventmagnitude) (Utilities|Struct|BreakGameplayEventData Gameplay%20Event%20Data))
  ;[0 1 0.2324] On both Client and Server, make sure the damage is from the storm and send to check for triggering warning / infliction
  (bind _asfort_player_pawn (Utilities|Casting|CastToFortPlayerPawn _target))
  (bind _output_get (Variables|Default|SetFortPlayerPawn _asfort_player_pawn))
  (Variables|Default|SetfCummulativeStormDamageForMatch (+ _eventmagnitude (Variables|Default|GetfCummulativeStormDamageForMatch))) ; Now Add the storm damage to a float
  ;[1 1 1] Setup delegates based on player pawn and minigame
  (Utilities|FlowControl|Sequence
    (:then_0
      (if (not (== (Minigame|GetMinigame _output_get :!class /Script/FortniteGame.MinigameBlueprintLibrary) (Variables|Default|GetStoredMinigame)))
        (CallFunction|UnbindMinigame)
        (bind _output_get_2 (Variables|Default|SetStoredMinigame (Minigame|GetMinigame _output_get :!class /Script/FortniteGame.MinigameBlueprintLibrary)))
        (Utilities|IsValid _output_get_2
          (:Is%20Valid
            (Minigame|BindEventtoMinigameEndedDelegate _output_get_2 (EventDispatchers|CreateEvent :!event Custom|ResetAll))
            (StormSickness|BindEventtoOnStormSicknessOptionsUpdated (Default|GetStormSicknessMutator _output_get_2) (EventDispatchers|CreateEvent :!event Custom|OnOptionsUpdated))))))
    (:then_1
      (Utilities|FlowControl|DoOnce false
        (:Completed
          (EventDispatchers|BindEventtoOnDied _output_get (EventDispatchers|CreateEvent :!event Custom|OnDied)))))))

(event Custom|ResetDeathEventBind
  ;[1 1 1] Setup delegates based on player pawn and minigame
  (Utilities|FlowControl|DoOnce false
    (:Completed
      (EventDispatchers|BindEventtoOnDied (stale Custom|SendStormSicknessGameplayEvent _output_get) (EventDispatchers|CreateEvent :!event Custom|OnDied)))))

(event Custom|CheckStormSicknessWarning
  ;[0.1194 0 1] Check if player should be warned for storm sickness
  (Utilities|EnabledCheck
    (:True
      (Math|Float|CompareFloat (Variables|Default|GetfCummulativeStormDamageForMatch) (ScalableFloat|GetValueatLevelwithContext (Variables|Default|GetfStormSicknessWarningLevel))
        (:>
          _)
        (:==
          _))
      (Utilities|FlowControl|DoOnce false
        (:Completed
          (Variables|Default|SetbPlayerWarned true)
          (bind _fortplayerpawn (Variables|Default|GetFortPlayerPawn)
            (:then
              (Ability|SendGameplayEventtoActor _fortplayerpawn "(TagName=\"Event.StormSicknessWarning\")" (Utilities|Struct|MakeGameplayEventData "(TagName=\"Event.StormSicknessWarning\")")))
            (:else
              (Development|PrintText "INVTEXT(\"Fort Player Pawn is Invalid!\")" true true "(R=0.000000,G=0.660000,B=1.000000,A=1.000000)" 105.0))))))))

(event Custom|ResetWarn
  ;[0.1194 0 1] Check if player should be warned for storm sickness
  (Variables|Default|SetbPlayerWarned false)
  (Utilities|FlowControl|DoOnce false
    (:Completed
      (Variables|Default|SetbPlayerWarned true)
      (bind _fortplayerpawn (Variables|Default|GetFortPlayerPawn)
        (:then
          (Ability|SendGameplayEventtoActor _fortplayerpawn "(TagName=\"Event.StormSicknessWarning\")" (Utilities|Struct|MakeGameplayEventData "(TagName=\"Event.StormSicknessWarning\")")))
        (:else
          (Development|PrintText "INVTEXT(\"Fort Player Pawn is Invalid!\")" true true "(R=0.000000,G=0.660000,B=1.000000,A=1.000000)" 105.0))))))

(event Custom|DisplayStormSicknessWarning
  ;[0.1194 0 1] Check if player should be warned for storm sickness
  (Utilities|FlowControl|DoOnce false
    (:Completed
      (Variables|Default|SetbPlayerWarned true)
      (bind _fortplayerpawn (Variables|Default|GetFortPlayerPawn)
        (:then
          (Ability|SendGameplayEventtoActor _fortplayerpawn "(TagName=\"Event.StormSicknessWarning\")" (Utilities|Struct|MakeGameplayEventData "(TagName=\"Event.StormSicknessWarning\")")))
        (:else
          (Development|PrintText "INVTEXT(\"Fort Player Pawn is Invalid!\")" true true "(R=0.000000,G=0.660000,B=1.000000,A=1.000000)" 105.0))))))

(event Custom|CheckAddStormSickness
  (bind _fcummulativestormdamageformatch (Variables|Default|GetfCummulativeStormDamageForMatch))
  ;[1 0 0.5161] Check to see if player should be inflicted with Storm Sickness
  (Utilities|EnabledCheck
    (:True
      (Math|Float|CompareFloat _fcummulativestormdamageformatch (ScalableFloat|GetValueatLevelwithContext (Variables|Default|GetfStormSicknessInflictionThreshold))
        (:>
          _)
        (:==
          _))
      (Math|Float|CompareFloat _fcummulativestormdamageformatch (ScalableFloat|GetValueatLevelwithContext (Variables|Default|GetfStormSicknessKillThreshold))
        (:>
          _)
        (:==
          _)
        (:<
          (Utilities|FlowControl|DoOnce false
            (:Completed
              (Variables|Default|SetbPlayerHasStormSickness true)
              (bind _fortplayerpawn (Variables|Default|GetFortPlayerPawn)
                (:then
                  (Ability|SendGameplayEventtoActor _fortplayerpawn "(TagName=\"Event.stormSicknessInflicted\")" (Utilities|Struct|MakeGameplayEventData "(TagName=\"Event.stormSicknessInflicted\")"))
                  (Utilities|IsServer
                    (:true
                      (bind _returnvalue (GameplayEffects|ApplyGameplayEffectToSelf :self (Ability|GetFortAbilitySystemComponent (Components|GetOwner)) :GameplayEffectClass (Variables|Default|GetGE_StormSickness) :Level 1.0))
                      (Variables|Default|SetSicknessEffect _returnvalue))) ; Only apply the GA on the server (client crashes)
                )
                (:else
                  (Development|PrintText "INVTEXT(\"Fort Player Pawn is Invalid!\")" true true "(R=0.000000,G=0.660000,B=1.000000,A=1.000000)" 105.0)))))))
      (CallFunction|KillSwitch))
    (:False
      (if (Networking|IsServer)
        (GameplayEffects|RemoveActiveGameplayEffect (Ability|GetFortAbilitySystemComponent (Components|GetOwner)) (Variables|Default|GetSicknessEffect) -1)))))

(event Custom|ResetAbility
  ;[1 0 0.5161] Check to see if player should be inflicted with Storm Sickness
  (if (Networking|IsServer)
    (GameplayEffects|RemoveActiveGameplayEffect (Ability|GetFortAbilitySystemComponent (Components|GetOwner)) (Variables|Default|GetSicknessEffect) -1)
    (Utilities|FlowControl|DoOnce false
      (:Completed
        (Variables|Default|SetbPlayerHasStormSickness true)
        (bind _fortplayerpawn (Variables|Default|GetFortPlayerPawn)
          (:then
            (Ability|SendGameplayEventtoActor _fortplayerpawn "(TagName=\"Event.stormSicknessInflicted\")" (Utilities|Struct|MakeGameplayEventData "(TagName=\"Event.stormSicknessInflicted\")"))
            (Utilities|IsServer
              (:true
                (bind _returnvalue (GameplayEffects|ApplyGameplayEffectToSelf :self (Ability|GetFortAbilitySystemComponent (Components|GetOwner)) :GameplayEffectClass (Variables|Default|GetGE_StormSickness) :Level 1.0))
                (Variables|Default|SetSicknessEffect _returnvalue))) ; Only apply the GA on the server (client crashes)
          )
          (:else
            (Development|PrintText "INVTEXT(\"Fort Player Pawn is Invalid!\")" true true "(R=0.000000,G=0.660000,B=1.000000,A=1.000000)" 105.0)))))))

(event Custom|ApplyStormSickness
  ;[1 0 0.5161] Check to see if player should be inflicted with Storm Sickness
  (Utilities|FlowControl|DoOnce false
    (:Completed
      (Variables|Default|SetbPlayerHasStormSickness true)
      (bind _fortplayerpawn (Variables|Default|GetFortPlayerPawn)
        (:then
          (Ability|SendGameplayEventtoActor _fortplayerpawn "(TagName=\"Event.stormSicknessInflicted\")" (Utilities|Struct|MakeGameplayEventData "(TagName=\"Event.stormSicknessInflicted\")"))
          (Utilities|IsServer
            (:true
              (bind _returnvalue (GameplayEffects|ApplyGameplayEffectToSelf :self (Ability|GetFortAbilitySystemComponent (Components|GetOwner)) :GameplayEffectClass (Variables|Default|GetGE_StormSickness) :Level 1.0))
              (Variables|Default|SetSicknessEffect _returnvalue))) ; Only apply the GA on the server (client crashes)
        )
        (:else
          (Development|PrintText "INVTEXT(\"Fort Player Pawn is Invalid!\")" true true "(R=0.000000,G=0.660000,B=1.000000,A=1.000000)" 105.0))))))

(event Custom|KillSwitch
  ;[1 0.007701 0] Eliminate the player if they're still somehow alive (to catch all remaining exploits / cheaters)
  (bind _fortplayerpawn (Variables|Default|GetFortPlayerPawn)
    (:then
      (FortPawn|ForceKill :self _fortplayerpawn)
      (Development|PrintText "INVTEXT(\"Player is Instant Killed!\")" true true "(R=0.000000,G=0.660000,B=1.000000,A=1.000000)" 1e+01))
    (:else
      (Development|PrintText "INVTEXT(\"Fort Player Pawn is Invalid!\")" true true "(R=0.000000,G=0.660000,B=1.000000,A=1.000000)" 105.0))))

(event Custom|Reset
  (params (ResetWarnAndSickness :type Boolean))
  (if ResetWarnAndSickness
    (Variables|Default|SetfCummulativeStormDamageForMatch 0.0)
    (CallFunction|ResetWarn)
    (else
      _))
  (CallFunction|ResetAbility)
  (CallFunction|ResetDeathEventBind))

(event Custom|ResetAll
  (CallFunction|Reset true)
  (CallFunction|ResetAbility)
  (CallFunction|ResetDeathEventBind))

(event Custom|FailedToActivatePassiveAbility_1)

(fn BuildDataRegistryResolverScope
  (return (Utilities|Array|MakeArray (SpatialGameplay|GetSpatialGameplayInterfaceFNameforActor (Components|GetOwner))) 1 true))

(fn GetCacheContextOverride
  (return 0))

(fn GetStormSicknessMutator
  (params (Minigame :type Fort%20Minigame%20Object%20Reference))
  (bind _returnvalue (Actor|GetComponentbyClass Minigame "/Script/FortniteGame.FortMutatorListComponent"))
  (Utilities|IsValid Minigame
    (:Is%20Valid
      (Utilities|IsValid _returnvalue
        (:Is%20Valid
          (bind _ascreative_mutator_storm_sickness_control (Utilities|Casting|CastToCreativeMutator_StormSicknessControl (Modifier|GetMutatorbyClass _returnvalue "/Game/Athena/Playlists/Creative/Mutators/CreativeMutator_StormSicknessControl.CreativeMutator_StormSicknessControl_C"))
            (:then
              _)
            (:CastFailed
              _))
          (return _ascreative_mutator_storm_sickness_control))
        (:Is%20Not%20Valid
          (return 0))))
    (:Is%20Not%20Valid
      (return 0))))

(fn IsRespawningAllowed
  (bind _asfort_game_state_zone (Utilities|Casting|CastToFortGameStateZone (Game|GetGameState)))
  (LocalVariables|SetGameState _asfort_game_state_zone)
  (bind _asfort_player_state (Utilities|Casting|CastToFortPlayerState (Components|GetOwner)))
  (LocalVariables|SetPlayerState _asfort_player_state)
  (return (Respawn|IsRespawningAllowed (LocalVariables|GetGameState) (LocalVariables|GetPlayerState))))

(fn OnRep_fCummulativeStormDamageForMatch
  (if (Variables|Default|GetbPlayerWarned)
    (CallFunction|CheckAddStormSickness)
    (else
      (CallFunction|CheckStormSicknessWarning)))
  (bind _fortplayerpawn (Variables|Default|GetFortPlayerPawn)
    (:then
      (Ability|SendGameplayEventtoActor _fortplayerpawn "(TagName=\"Event.stormSickness.ThresholdUpdate\")" (Utilities|Struct|MakeGameplayEventData "(TagName=\"Event.stormSickness.ThresholdUpdate\")" 0 0 0 0 "()" "(GameplayTags=)" "(GameplayTags=)" (Math|Float|Clamp(Float) (/ (Variables|Default|GetfCummulativeStormDamageForMatch) (ScalableFloat|GetValueatLevel (Variables|Default|GetfStormSicknessInflictionThreshold)))))) ; Send out update of percentage of threshold
    )
    (:else)))

(fn ValidatedGetFortPlayerPawn
  (bind _fortplayerpawn (Variables|Default|GetFortPlayerPawn))
  (Utilities|IsValid _fortplayerpawn
    (:Is%20Valid
      (return _fortplayerpawn))
    (:Is%20Not%20Valid
      (return (PlayerState|GetCurrentPawn (Utilities|Casting|CastToFortPlayerState (Components|GetOwner)))))))
